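# Sound cue for the dismiss action in the AI companion controls UI.
# A single wave player plays the AI_CompanionControl_Dismiss sound with boosted volume in the non-critical in-game UI sound class.
# A concurrency override limits it to one playing instance at a time.
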SoundCue /Game/Sounds/Fort_UI_Sounds/AI_CompanionControls/AI_CompanionControl_Dismiss_Cue
root=WavePlayer_0
nodes [{id=WavePlayer_0, properties {soundWaveAssetPtr=&/AI_CompanionControl_Dismiss}}]
properties {
 bOverrideConcurrency=true
 maxDistance=2097152
 soundClassObject=&Fort_UI_InGame_NonCritical
 volumeMultiplier=2.5
 concurrencyOverrides {maxCount=1}
}
